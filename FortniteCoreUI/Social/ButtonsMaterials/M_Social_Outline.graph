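# UI material for Social buttons that draws an animated, rift-energy outline and glow with alpha-composite blending.
# A rounded-mask function with skew supplies shape and outline SDFs for antialiased edges, with padding corrected for aspect ratio.
# Panned polar noise with explicit derivatives, screen-space flow-map noise and a rotated sheen build the energy effects.
# Anim parameters drive hover and thickness, while a static switch swaps oscillating rift colors for flat color.
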Material /FortniteCoreUI/Social/ButtonsMaterials/M_Social_Outline
graph {
 properties {blendMode=BLEND_AlphaComposite, materialDomain=MD_UI, pixelDepthOffsetMode=PDOM_Legacy, usedShadingModels=Default Lit}
 nodes [
  Multiply(A=NamedRerouteUsage(), B=Multiply(A=VertexColor().A, B=NamedRerouteUsage()))
  n1 = Constant()
  n2 = NamedRerouteUsage()
  n3 = LinearInterpolate(A=Constant2Vector() {g=2, r=4}, B=Constant2Vector() {g=4, r=2}, Alpha=NamedRerouteUsage())
  n4 = MaterialFunctionCall!(Dimensions=MaterialFunctionCall(UVs=NamedRerouteUsage(), "Rotation Angle (0-1)"=NamedRerouteUsage()) {materialFunction=&CustomRotator}.Rotated Values) {materialFunction=/FortUILibrary/MaterialFunctions/Basics/MF_UI_GetAspectRatio}
  n5 = Divide(
   A=NamedRerouteUsage()
   B=LinearInterpolate(A=ComponentMask(None=@n4.Max Ratio) {g=true}, B=ComponentMask(None=@n4.Max Ratio) {r=true}, Alpha=NamedRerouteUsage())
  )
  n6 = Add(
   A=MaterialFunctionCall(
    UVs=MaterialFunctionCall(UVs=@n2, Texture Scale=@n3) {materialFunction=&ScaleUVsByCenter}.UVs
    "Rotation Angle (0-1)"=NamedRerouteUsage()
   ) {materialFunction=&CustomRotator}.Rotated Values
   B=AppendVector(A=LinearInterpolate(Alpha=@n5) {constA=2, constB=0}, B=Constant())
  )
  n7 = Divide(A=Length(None=ConstantBiasScale(None=@n6) {bias=-0.5, scale=2}) {is=class MaterialExpressionLength})
  n8 = Power(
   Base=MaterialFunctionCall!(1=@n1, 2=Constant() {r=1}, 3=@n1, 4, A=Saturate(None=@n7)) {materialFunction=/Engine/Functions/Engine_MaterialFunctions02/Lerp_Multiple_Float}.Lerp 3 Inputs
   Exp=LinearInterpolate(Alpha=ScalarParameter() {parameterName=Sheen_Falloff}) {constA=5, constB=10}
  ) {constExponent=10}
  Multiply(A=@n8) {constB=0.3}
  n9 = VectorParameter() {parameterName=Corner Radius, defaultValue=<rgba 0.5 0.5 0.5 0.5>}
  n10 = LinearInterpolate(
   B=ScalarParameter() {defaultValue=4, group=Outline, parameterName=Outline Thickness}
   Alpha=ScalarParameter() {defaultValue=1, group=Anim, parameterName=Anim_OutlineThickness, sliderMax=1}
  )
  n11 = Saturate(None=@n10)
  n12 = NamedRerouteDeclaration(None=LinearInterpolate(B=@n10, Alpha=Floor(None=@n11))) {name=Outline Thickness, nodeColor=<rgba 0 0 0 1>}
  n13 = MaterialFunctionCall(
   UV=NamedRerouteUsage()
   Size=NamedRerouteUsage()
   Corner Radius=AppendVector(A=@n9.RGB, B=@n9.RGB)
   Radius in Pixels=ScalarParameter() {parameterName=Radius in Pixels, sliderMax=1}
   Skew Angle=NamedRerouteDeclaration(None=ScalarParameter() {parameterName=Skew Angle, sliderMax=360}) {name=Skew Angle, nodeColor=<rgba 0 0 0 1>}
   Outline Thickness=@n12
  ) {materialFunction=&MF_UI_CoreUI_RoundedMask}
  n14 = Constant() {r=0.01}
  n15 = Saturate(None=@n12)
  NamedRerouteDeclaration(None=MaterialFunctionCall(
    DistanceField=NamedRerouteDeclaration(None=@n13.Outline SDF) {name=SDF Outline, nodeColor=<rgba 0 0 0 1>}
    Smoothing=LinearInterpolate(B=@n14, Alpha=@n15)
   ) {materialFunction=&MF_UI_AntiAliasing}.Inner) {name=Outline Mask, nodeColor=<rgba 0 0 0 1>}
  NamedRerouteDeclaration(None=ScalarParameter() {defaultValue=0.25, parameterName=SheenRotation}) {name=Sheen Rotation, nodeColor=<rgba 0 0 0 1>}
  n16 = NamedRerouteUsage()
  n17 = MaterialFunctionCall() {materialFunction=&MF_UI_GetAspectRatio}
  n18 = ComponentMask(None=@n17.Max Ratio) {r=true}
  n19 = ComponentMask(None=@n17.Max Ratio) {g=true}
  n20 = Divide(B=Max(A=@n18, B=@n19))
  n21 = Add(A=@n16, B=Multiply(A=OneMinus(None=@n20), B=NamedRerouteUsage()) {constB=0.25})
  n22 = MaterialFunctionCall!(A=@n18, B=@n19) {materialFunction=/FortUILibrary/MaterialFunctions/Basics/MF_UI_BooleanOperations}
  NamedRerouteDeclaration(None=LinearInterpolate(A=AppendVector(A=@n16, B=@n21), B=AppendVector(A=@n21, B=@n16), Alpha=@n22.IsGreater)) {name=Size with Padding, nodeColor=<rgba 0 0 0 1>}
  n23 = NamedRerouteUsage()
  n24 = Constant() {r=0.5}
  NamedRerouteDeclaration(None=Multiply(A=Modulo!(A=@n23, B=@n24)) {constB=4}) {name=Is Vertical, nodeColor=<rgba 0 0 0 1>}
  n25 = MaterialFunctionCall() {materialFunction=&GetUserInterfaceUV}
  NamedRerouteDeclaration(None=@n25.Pixel Size) {name=Pixel Size, nodeColor=<rgba 0 0 0 1>}
  n26 = Multiply(A=@n8, B=LinearInterpolate(Alpha=NamedRerouteUsage()) {constA=0.333})
  n27 = NamedRerouteUsage()
  NamedRerouteDeclaration(None=Multiply(A=Multiply(A=@n26, B=@n27), B=ScalarParameter() {defaultValue=0.5, group=Alpha, parameterName=Sheen_Alpha, sliderMax=1})) {name=Sheen Effect, nodeColor=<rgba 0 0 0 1>}
  NamedRerouteDeclaration(None=ScalarParameter() {defaultValue=0.5, group=Anim, parameterName=Anim_Highlight, sliderMax=1}) {name=Animation Highlight, nodeColor=<rgba 0 0 0 1>}
  n28 = NamedRerouteUsage()
  n29 = Sine(None=Multiply(A=Time()) {constB=0.125})
  n30 = LinearInterpolate(
   A=NamedRerouteUsage()
   B=AppendVector(A=OneMinus(None=@n28), B=@n28)
   Alpha=LinearInterpolate(Alpha=ConstantBiasScale(None=@n29)) {constA=0.95, constB=0.85}
  )
  n31 = NamedRerouteUsage()
  n32 = MaterialFunctionCall!(Vector or UVs=NamedRerouteDeclaration(None=MaterialFunctionCall!(Value=@n30, Scale=@n31, Height=@n31) {materialFunction=/FortUILibrary/MaterialFunctions/Basics/MF_UI_PingPong}.Result) {name=PrepPolarUVs, nodeColor=<rgba 0 0 0 1>}) {materialFunction=/FortUILibrary/MaterialFunctions/Basics/MF_UI_CheaperVectorToRadial}
  n33 = ComponentMask(None=@n32.Radial Coordinates) {r=true}
  n34 = ComponentMask(None=@n32.Radial Coordinates) {g=true}
  n35 = AppendVector(A=Multiply(A=@n33) {constB=3}, B=Multiply(A=@n34) {constB=2.75})
  n36 = Time()
  n37 = Constant() {r=-0.05}
  n38 = AppendVector(A=Constant() {r=0.1}, B=@n37)
  n39 = Multiply(A=@n34) {constB=1.5}
  n40 = NamedRerouteUsage()
  n41 = TextureSample!(
   UVs=Panner(Coordinate=AppendVector(A=@n33, B=@n39), Speed=@n37)
   Tex=NamedRerouteDeclaration(None=TextureObject() {desc=Rift Noise Texture, samplerType=SAMPLERTYPE_Grayscale, texture=/FortUILibrary/UIKit/Buttons/Textures/LowResBlurredNoise}) {name=Rift Noise Texture, nodeColor=<rgba 0 0 0 1>}
   "DDX(UVs)"=NamedRerouteDeclaration(None=DDX(Value=@n40)) {name=Polar DDX, nodeColor=<rgba 0 0 0 1>}
   "DDY(UVs)"=NamedRerouteDeclaration(None=DDY(Value=@n40)) {name=Polar DDY, nodeColor=<rgba 0 0 0 1>}
   Apply View MipBias
  ) {mipValueMode=TMVM_Derivative}
  n42 = Constant() {r=0.5}
  n43 = TextureSample(
   UVs=Add(A=Panner(Coordinate=@n35, Time=@n36, Speed=@n38), B=Multiply(A=@n41.R, B=@n42))
   Tex=NamedRerouteUsage()
   "DDX(UVs)"=NamedRerouteUsage()
   "DDY(UVs)"=NamedRerouteUsage()
  ) {mipValueMode=TMVM_Derivative}
  NamedRerouteDeclaration(None=SmoothStep(Value=DotProduct(A=@n43.RGB, B=@n41.RGB))) {name=Rift Pulse Mask, nodeColor=<rgba 0 0 0 1>}
  n44 = Rotator(Coordinate=NamedRerouteUsage(), Time=Multiply(A=Time()) {constB=0.333}) {speed=6.2832}
  NamedRerouteDeclaration(None=LinearInterpolate(
    A=Constant3Vector() {constant=<rgba 1 1 1 1>}
    B=MaterialFunctionCall!(PrimaryUV, SecondaryUV, UVMix, PixelSize) {materialFunction=/FortUILibrary/MaterialFunctions/CoreUI/MF_UI_CoreUI_OscillatingColor}.Result
    Alpha=ComponentMask(None=@n44) {r=true}
   ) {constA=1}) {name=RiftOscillatingColors, nodeColor=<rgba 0 0 0 1>}
  NamedRerouteDeclaration(None=OneMinus(None=NamedRerouteUsage())) {name=Outline Size, nodeColor=<rgba 0 0 0 1>}
  NamedRerouteDeclaration(None=@n25.Normalized UV) {name=NormalizedUV, nodeColor=<rgba 0 0 0 1>}
  NamedRerouteDeclaration(None=ScalarParameter() {defaultValue=16, group=Padding, parameterName=Padding}) {name=Padding, nodeColor=<rgba 0 0 0 1>}
  n45 = LinearInterpolate(
   A=NamedRerouteUsage()
   B=LinearInterpolate(A=Constant3Vector() {constant=<rgba 0.667 1 1 1>}, B=NamedRerouteUsage(), Alpha=Constant() {desc="RiftEnergyOscillatingColorAmount`", r=0.5})
   Alpha=NamedRerouteUsage()
  )
  n46 = LinearInterpolate(
   A=LinearInterpolate(A=NamedRerouteUsage(), B=Constant3Vector() {constant=<rgba 0.5 0.5 1 1>}, Alpha=Constant() {desc=RiftEnergyColorTintMix, r=0.8})
   B=Constant3Vector() {desc=Outline Color, constant=<rgba 1 1 1 1>}
   Alpha=OneMinus(None=NamedRerouteUsage())
  )
  n47 = NamedRerouteUsage()
  NamedRerouteDeclaration(None=StaticSwitchParameter(
    True=VectorParameter() {group=Color, parameterName=Color_Outline, defaultValue=<rgba 1 1 1 1>}.RGB
    False=LinearInterpolate(A=@n45, B=@n46, Alpha=@n47)
   ) {parameterName=UseFlatColor}) {name=FinalColor, nodeColor=<rgba 0 0 0 1>}
  n48 = NamedRerouteUsage()
  n49 = ScalarParameter() {defaultValue=1, group=Alpha, parameterName=Glow_Alpha, sliderMax=1}
  n50 = ScalarParameter() {defaultValue=1, group=Alpha, parameterName=Glow_Alpha, sliderMax=1}
  n51 = Sine(None=Multiply(A=Time(), B=Constant() {desc=Glow_Alpha_Change_Speed, r=0.25}))
  n52 = LinearInterpolate(
   B=LinearInterpolate(A=Add(A=@n50) {constB=0.05}, B=Add(A=@n50) {constB=-0.05}, Alpha=ConstantBiasScale(None=@n51))
   Alpha=ScalarParameter() {defaultValue=1, group=Anim, parameterName=Anim_Hover_Fill, sliderMax=1}
  ) {constA=1}
  n53 = NamedRerouteUsage()
  n54 = Multiply(A=NamedRerouteUsage(), B=ScalarParameter() {defaultValue=1, group=Anim, parameterName=Anim_Hover_Energy, sliderMax=1})
  n55 = LinearInterpolate(Alpha=NamedRerouteUsage()) {constA=0.75}
  n56 = Multiply(A=LinearInterpolate(A=Multiply(A=@n52, B=@n53), Alpha=Multiply(A=@n54, B=@n55)), B=NamedRerouteUsage())
  n57 = LinearInterpolate(A=NamedRerouteDeclaration(None=Multiply(A=@n49, B=@n56)) {name=Rift Glow Alpha, nodeColor=<rgba 0 0 0 1>}, Alpha=NamedRerouteUsage())
  NamedRerouteDeclaration(None=Multiply(A=LinearInterpolate(A=@n48, Alpha=@n57), B=ScalarParameter() {defaultValue=1, group=Alpha, parameterName=FInalAlpha, sliderMax=1})) {name=FinalAlpha, nodeColor=<rgba 0 0 0 1>}
  NamedRerouteDeclaration(None=@n13.Shape Mask) {name=Shape Mask, nodeColor=<rgba 0 0 0 1>}
  n58 = NamedRerouteUsage()
  n59 = SmoothStep(Min=Multiply(A=@n58) {constB=-1}, Max=@n58, Value=NamedRerouteUsage()) {constMax=0.05}
  n60 = NamedRerouteUsage()
  NamedRerouteDeclaration(None=Multiply(A=OneMinus(None=@n59), B=Saturate(None=@n60))) {name=Glow Mask, nodeColor=<rgba 0 0 0 1>}
  n61 = NamedRerouteDeclaration(None=LinearInterpolate(Alpha=NamedRerouteUsage()) {constA=0.5}) {name=Reduced Flow Noise, nodeColor=<rgba 0 0 0 1>}
  n62 = NamedRerouteUsage()
  NamedRerouteDeclaration(None=Multiply(A=Multiply(A=@n61, B=@n62), B=ScalarParameter() {defaultValue=1, group=Anim, parameterName=Anim_Hover_Outline, sliderMax=1})) {name=Outline Alpha, nodeColor=<rgba 0 0 0 1>}
  NamedRerouteDeclaration(None=MaterialFunctionCall!(Value PX=NamedRerouteUsage(), UV Pixel Size=NamedRerouteUsage(), X or Y=@n22.IsGreater) {materialFunction=/FortUILibrary/MaterialFunctions/Basics/MF_UI_PixelsToNormalized}.Normalized) {name=Normalized Padding, nodeColor=<rgba 0 0 0 1>}
  n63 = ScreenPosition()
  n64 = Add(A=NamedRerouteUsage())
  n65 = Time()
  n66 = Multiply(A=@n65) {constB=0.05}
  n67 = Add(A=Divide(A=@n63.ViewportUV, B=Multiply(A=@n64) {constB=2}), B=AppendVector(A=Multiply(A=@n65) {constB=-0.1}, B=@n66))
  n68 = NamedRerouteUsage()
  n69 = Divide(A=@n63.ViewportUV, B=Multiply(A=@n64) {constB=3})
  n70 = AppendVector(A=Multiply(A=@n65) {constB=0.025}, B=@n66)
  n71 = AppendVector(
   A=TextureSample!(UVs=NamedRerouteDeclaration(None=@n67) {name=PanningUVs, nodeColor=<rgba 0 0 0 1>}, Tex=@n68, Apply View MipBias).R
   B=TextureSample(UVs=Add(A=@n69, B=@n70), Tex=@n68).R
  )
  n72 = Constant() {r=0.2}
  n73 = MaterialFunctionCall!(
   UVs=NamedRerouteUsage()
   Texture Object=NamedRerouteUsage()
   Grayscale Flow Map=@n71
   Flow Direction=OneMinus(None=ComponentMask(None=@n71) {r=true})
   Flow Intensity=@n72
   Flow Speed=@n72
  ) {materialFunction=/FortUILibrary/MaterialFunctions/Effects/MF_UI_GrayscaleFlowMap_New}
  NamedRerouteDeclaration(None=SmoothStep(Value=ComponentMask(None=@n73.Distorted Texture) {r=true})) {name=Flowing Noise Mask, nodeColor=<rgba 0 0 0 1>}
 ]
}
